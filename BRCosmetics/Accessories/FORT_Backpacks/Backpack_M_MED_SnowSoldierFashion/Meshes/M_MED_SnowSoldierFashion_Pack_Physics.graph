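# Physics asset for the SnowSoldierFashion backpack cosmetic, defining collision bodies and joints for its dangling straps and carabiner.
# Capsule chains form a three-segment big strap and a two-segment small strap, each anchored to a small sphere master body, while a two-sphere carabiner chain hangs from the pack capsule.
# Angle-limited constraints keep twist tight, and large unconstrained spine capsules supply torso collision volumes.
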PhysicsAsset /BRCosmetics/Accessories/FORT_Backpacks/Backpack_M_MED_SnowSoldierFashion/Meshes/M_MED_SnowSoldierFashion_Pack_Physics
bodies {
 big_trap_MSTR [{shapeType=Sphere, radius=0.5, shapeName=big_trap_MSTR_sphere}]
 dyn_big_strap_01 [{shapeType=Capsule, length=1.4, radius=0.9, shapeName=dyn_big_strap_01_capsule, center=<2.0219 0 0>, rotation=<rot 90 0 0>}]
 dyn_big_strap_02 [{shapeType=Capsule, length=1.4, radius=0.9, shapeName=dyn_big_strap_01_capsule, center=<1.7802 0 0>, rotation=<rot 90 0 0>}]
 dyn_big_strap_03 [{shapeType=Capsule, length=5, radius=0.9, shapeName=dyn_big_strap_01_capsule, center=<3.5747 0 0>, rotation=<rot 90 0 0>}]
 dyn_carabine_01 [{shapeType=Sphere, radius=0.505, shapeName=dyn_carabine_01_sphere, center=<0.5195 0.0423 0.1718>}]
 dyn_carabine_02 [{shapeType=Sphere, radius=1.9794, shapeName=dyn_carabine_02_sphere, center=<1.4496 0.0079 0.0815>}]
 dyn_small_strap_01 [{shapeType=Capsule, length=1.4, radius=0.9, shapeName=dyn_big_strap_01_capsule, center=<2.0219 0 0>, rotation=<rot 90 0 0>}]
 dyn_small_strap_02 [{shapeType=Capsule, length=1.4, radius=0.9, shapeName=dyn_big_strap_01_capsule, center=<2.0219 0 0>, rotation=<rot 90 0 0>}]
 pack [{shapeType=Capsule, length=23, radius=3, center=<1.5796 -0.1167 -2.267>, rotation=<rot 4.8865 -90 42.0395>}]
 small_strap_MSTR [{shapeType=Sphere, radius=0.5, shapeName=small_strap_MSTR_sphere}]
 spine_01 [{shapeType=Capsule, length=14, radius=15, shapeName=spine_01_capsule, center=<0 -1 0>}]
 spine_04 [{shapeType=Capsule, length=14, radius=14, shapeName=spine_01_capsule, center=<0 -1 0>}]
}
constraints [
 {bone1Name=dyn_big_strap_01, bone2Name=big_trap_MSTR, swing1LimitDegrees=45, swing2LimitDegrees=20, twistLimitDegrees=1}
 {bone1Name=dyn_big_strap_02, bone2Name=dyn_big_strap_01, swing1LimitDegrees=20, swing2LimitDegrees=5, twistLimitDegrees=1}
 {bone1Name=dyn_big_strap_03, bone2Name=dyn_big_strap_02, swing1LimitDegrees=30, swing2LimitDegrees=0, twistLimitDegrees=1}
 {bone1Name=dyn_small_strap_02, bone2Name=dyn_small_strap_01, swing1LimitDegrees=50, swing2LimitDegrees=20, twistLimitDegrees=1}
 {bone1Name=dyn_small_strap_01, bone2Name=small_strap_MSTR, swing1LimitDegrees=40, swing2LimitDegrees=40, twistLimitDegrees=1}
 {bone1Name=dyn_carabine_02, bone2Name=dyn_carabine_01, swing1LimitDegrees=30, swing2LimitDegrees=30, twistLimitDegrees=5}
 {bone1Name=dyn_carabine_01, bone2Name=pack, swing1LimitDegrees=30, swing2LimitDegrees=2, twistLimitDegrees=2}
]
