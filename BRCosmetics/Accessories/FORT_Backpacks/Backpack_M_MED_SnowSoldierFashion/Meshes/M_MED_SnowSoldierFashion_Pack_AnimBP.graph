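# Animation Blueprint for the SnowSoldierFashion backpack, fitting the pack to the character's body and simulating its dangling parts.
# It copies BodyMesh spine transforms, constrains and offsets the pack by body type, and adds skydive and clavicle-driven corrections only when the body mesh name contains SnowSoldierFashion.
# bRigidBodyEnabled selects RigidBody physics, otherwise chained AnimDynamics springs move straps and carabiners.
# The zipper always uses AnimDynamics after that blend.
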AnimBlueprint /BRCosmetics/Accessories/FORT_Backpacks/Backpack_M_MED_SnowSoldierFashion/Meshes/M_MED_SnowSoldierFashion_Pack_AnimBP
target_skeleton=&/M_MED_SnowSoldierFashion_Pack_Skeleton
graphs [
 AnimGraph {nodes [
   "Transform (Modify) Bone - Bone: pack"(Translation=Get backpack_translation(), Rotation=Get backpack_rotate()) {
    rotationMode=BMM_Additive
    rotationSpace=BCS_BoneSpace
    translationMode=BMM_Additive
    translationSpace=BCS_BoneSpace
    alphaBoolBlend {blendOption=Linear}
    boneToModify {boneName=pack}
   }
   n1 = RigidBody!(
    ComponentPose=Local To Component(LocalPose=Use cached pose 'PreSim2'())
    Alpha
    OverrideWorldGravity:Vector=Get Gravity Override()
    ExternalForce
   ) -> (Pose) {
    bClampLinearTranslationLimitToRefPose=true
    bOverrideWorldGravity=true
    overridePhysicsAsset=&/M_MED_SnowSoldierFashion_Pack_Physics
    simulationSpace=BaseBoneSpace
    alphaBoolBlend {blendOption=Linear}
    baseBoneRef {boneName=pack}
    componentAppliedLinearAccClamp=<3000 3000 3000>
    componentLinearAccScale=<0.4 0.4 0.3>
    componentLinearVelScale=<0.4 0.4 0.5>
    overrideWorldGravity {z=-980}
   }
   n2 = Component To Local(ComponentPose=@n1)
   n3 = Anim Dynamics - Bone: dyn_small_strap_01!(
    ComponentPose=Local To Component(LocalPose=Use cached pose 'PreSim2'())
    bAlphaBoolEnabled:bool=True
    GravityOverride:Vector="0,0,-980"
    ExternalForce="0,0,0"
   ) -> (Pose) {
    alphaInputType=Bool
    angularDampingOverride=0.9999
    angularSpringConstant=1000
    bAngularSpring=true
    bOverrideAngularDamping=true
    bUseGravityOverride=true
    simulationSpace=RootRelative
    alphaBoolBlend {blendInTime=0.1, blendOption=Linear, blendOutTime=0.1}
    boundBone {boneName=dyn_small_strap_01}
    componentAppliedLinearAccClamp=<3000 3000 3000>
    componentLinearAccScale=<0.4 0.5 0.5>
    componentLinearVelScale=<0.5 0.5 0.5>
    gravityOverride {z=-500}
    physicsBodyDefinitions [{
      collisionType=CustomSphere
      sphereCollisionRadius=2.5
      boundBone {boneName=dyn_small_strap_01}
      boxExtents=<8 8 8>
      constraintSetup {angularLimitsMax=<0 5 4>, angularLimitsMin=<0 -5 -35>, angularTarget=<0 0 0>}
      localJointOffset=<2 0 0>
     }]
    relativeSpaceBone {boneName=pack}
   }
   n4 = Anim Dynamics - Bone: dyn_small_strap_02!(ComponentPose=@n3, bAlphaBoolEnabled:bool=True, GravityOverride:Vector="0,0,-980", ExternalForce="0,0,0") -> (Pose) {
    alphaInputType=Bool
    angularDampingOverride=0.9999
    angularSpringConstant=1000
    bAngularSpring=true
    bOverrideAngularDamping=true
    bUseGravityOverride=true
    simulationSpace=RootRelative
    alphaBoolBlend {blendInTime=0.1, blendOption=Linear, blendOutTime=0.1}
    boundBone {boneName=dyn_small_strap_02}
    componentAppliedLinearAccClamp=<3000 3000 3000>
    componentLinearAccScale=<0.4 0.5 0.5>
    componentLinearVelScale=<0.5 0.5 0.5>
    gravityOverride {z=-500}
    physicsBodyDefinitions [{
      collisionType=CustomSphere
      sphereCollisionRadius=2.5
      boundBone {boneName=dyn_small_strap_02}
      boxExtents=<8 8 8>
      constraintSetup {angularLimitsMax=<0 5 4>, angularLimitsMin=<0 -5 -35>, angularTarget=<0 0 0>}
      localJointOffset=<2 0 0>
     }]
    relativeSpaceBone {boneName=pack}
   }
   n5 = Anim Dynamics - Bone: dyn_big_strap_01!(ComponentPose=@n4, bAlphaBoolEnabled:bool=True, GravityOverride:Vector="0,0,-980", ExternalForce="0,0,0") -> (Pose) {
    alphaInputType=Bool
    angularDampingOverride=0.9999
    angularSpringConstant=1000
    bAngularSpring=true
    bOverrideAngularDamping=true
    bUseGravityOverride=true
    simulationSpace=RootRelative
    alphaBoolBlend {blendInTime=0.1, blendOption=Linear, blendOutTime=0.1}
    boundBone {boneName=dyn_big_strap_01}
    componentAppliedLinearAccClamp=<3000 3000 3000>
    componentLinearAccScale=<0.4 0.5 0.5>
    componentLinearVelScale=<0.5 0.5 0.5>
    gravityOverride {z=-500}
    physicsBodyDefinitions [{
      collisionType=CustomSphere
      sphereCollisionRadius=2.5
      boundBone {boneName=dyn_big_strap_01}
      boxExtents=<8 8 8>
      constraintSetup {angularLimitsMax=<0 5 4>, angularLimitsMin=<0 -5 -25>, angularTarget=<0 0 0>}
      localJointOffset=<2 0 0>
     }]
    relativeSpaceBone {boneName=pack}
   }
   n6 = Anim Dynamics - Bone: dyn_big_strap_02!(ComponentPose=@n5, bAlphaBoolEnabled:bool=True, GravityOverride:Vector="0,0,-980", ExternalForce="0,0,0") -> (Pose) {
    alphaInputType=Bool
    angularDampingOverride=0.9999
    angularSpringConstant=1000
    bAngularSpring=true
    bOverrideAngularDamping=true
    bUseGravityOverride=true
    simulationSpace=RootRelative
    alphaBoolBlend {blendInTime=0.1, blendOption=Linear, blendOutTime=0.1}
    boundBone {boneName=dyn_big_strap_02}
    componentAppliedLinearAccClamp=<3000 3000 3000>
    componentLinearAccScale=<0.4 0.5 0.5>
    componentLinearVelScale=<0.5 0.5 0.5>
    gravityOverride {z=-500}
    physicsBodyDefinitions [{
      collisionType=CustomSphere
      sphereCollisionRadius=2.5
      boundBone {boneName=dyn_big_strap_02}
      boxExtents=<8 8 8>
      constraintSetup {angularLimitsMax=<0 5 4>, angularLimitsMin=<0 -5 -25>, angularTarget=<0 0 0>}
      localJointOffset=<2 0 0>
     }]
    relativeSpaceBone {boneName=pack}
   }
   n7 = Anim Dynamics - Bone: dyn_big_strap_03!(ComponentPose=@n6, bAlphaBoolEnabled:bool=True, GravityOverride:Vector="0,0,-980", ExternalForce="0,0,0") -> (Pose) {
    alphaInputType=Bool
    angularDampingOverride=0.9999
    angularSpringConstant=1000
    bAngularSpring=true
    bOverrideAngularDamping=true
    bUseGravityOverride=true
    simulationSpace=RootRelative
    alphaBoolBlend {blendInTime=0.1, blendOption=Linear, blendOutTime=0.1}
    boundBone {boneName=dyn_big_strap_03}
    componentAppliedLinearAccClamp=<3000 3000 3000>
    componentLinearAccScale=<0.4 0.5 0.5>
    componentLinearVelScale=<0.5 0.5 0.5>
    gravityOverride {z=-500}
    physicsBodyDefinitions [{
      collisionType=CustomSphere
      sphereCollisionRadius=2.5
      boundBone {boneName=dyn_big_strap_03}
      boxExtents=<8 8 8>
      constraintSetup {angularLimitsMax=<0 5 4>, angularLimitsMin=<0 -5 -25>, angularTarget=<0 0 0>}
      localJointOffset=<2 0 0>
     }]
    relativeSpaceBone {boneName=pack}
   }
   n8 = Anim Dynamics - Bone: dyn_carabine_01!(ComponentPose=@n7, bAlphaBoolEnabled:bool=True, GravityOverride:Vector="0,0,-980", ExternalForce="0,0,0") -> (Pose) {
    alphaInputType=Bool
    angularDampingOverride=0.8
    angularSpringConstant=1000
    bAngularSpring=true
    bOverrideAngularDamping=true
    bUseGravityOverride=true
    simulationSpace=RootRelative
    alphaBoolBlend {blendInTime=0.1, blendOption=Linear, blendOutTime=0.1}
    boundBone {boneName=dyn_carabine_01}
    componentAppliedLinearAccClamp=<3000 3000 3000>
    componentLinearAccScale=<0.4 0.4 0.5>
    componentLinearVelScale=<0.4 0.4 0.5>
    gravityOverride {z=-500}
    physicsBodyDefinitions [{
      collisionType=CustomSphere
      sphereCollisionRadius=2.5
      boundBone {boneName=dyn_carabine_01}
      boxExtents=<3 3 3>
      constraintSetup {angularLimitsMax=<0 0 0>, angularLimitsMin=<0 0 -25>, angularTarget=<0 0 0>}
      localJointOffset=<2 0 0>
     }]
    relativeSpaceBone {boneName=pack}
   }
   n9 = Anim Dynamics - Bone: dyn_carabine_02!(ComponentPose=@n8, bAlphaBoolEnabled:bool=True, GravityOverride:Vector="0,0,-980", ExternalForce="0,0,0") -> (Pose) {
    alphaInputType=Bool
    angularDampingOverride=0.9999
    angularSpringConstant=1000
    bAngularSpring=true
    bOverrideAngularDamping=true
    bUseGravityOverride=true
    simulationSpace=RootRelative
    alphaBoolBlend {blendInTime=0.1, blendOption=Linear, blendOutTime=0.1}
    boundBone {boneName=dyn_carabine_02}
    componentAppliedLinearAccClamp=<3000 3000 3000>
    componentLinearAccScale=<0.3 0.3 0.3>
    componentLinearVelScale=<0.3 0.3 0.3>
    gravityOverride {z=-500}
    physicsBodyDefinitions [{
      collisionType=CustomSphere
      sphereCollisionRadius=2.5
      boundBone {boneName=dyn_carabine_02}
      boxExtents=<3 3 3>
      constraintSetup {angularLimitsMax=<0 0 0>, angularLimitsMin=<0 0 -35>, angularTarget=<0 0 0>}
      localJointOffset=<3 0 0>
     }]
    relativeSpaceBone {boneName=pack}
   }
   n10 = Component To Local(ComponentPose=@n9)
   n11 = Get bRigidBodyEnabled()
   n12 = Anim Dynamics - Bone: dyn_zipper_01!(
    ComponentPose=Local To Component(LocalPose=Blend Poses by bool(BlendPose_0=@n2, BlendPose_1=@n10, bActiveValue=@n11) {blendType=Linear})
    bAlphaBoolEnabled:bool=True
    GravityOverride:Vector="0,0,-980"
    ExternalForce="0,0,0"
   ) -> (Pose) {
    alphaInputType=Bool
    angularDampingOverride=0.9999
    angularSpringConstant=1000
    bAngularSpring=true
    bOverrideAngularDamping=true
    bUseGravityOverride=true
    simulationSpace=RootRelative
    alphaBoolBlend {blendInTime=0.1, blendOption=Linear, blendOutTime=0.1}
    boundBone {boneName=dyn_zipper_01}
    componentAppliedLinearAccClamp=<3000 3000 3000>
    componentLinearAccScale=<0.3 0.3 0.3>
    componentLinearVelScale=<0.3 0.3 0.3>
    gravityOverride {z=-500}
    physicsBodyDefinitions [{
      collisionType=CustomSphere
      sphereCollisionRadius=2.5
      boundBone {boneName=dyn_zipper_01}
      boxExtents=<5 5 5>
      constraintSetup {angularLimitsMax=<0 80 15>, angularLimitsMin=<0 -30 -15>, angularTarget=<0 0 0>}
      localJointOffset=<3 0 0>
     }]
    relativeSpaceBone {boneName=pack}
   }
   Output Pose(Result=Component To Local(ComponentPose=@n12)) {layerGroup=None}
   n13 = "Transform (Modify) Bone - Bone: spine_01"(
    ComponentPose=Local To Component(LocalPose=Input Pose())
    Alpha=1
    Translation=Get spine_01 translation() -> (spine_01 translation:Vector)
    Rotation=Get spine_01rotation() -> (spine_01rotation:Rotator)
    Scale="1,1,1"
   ) {
    rotationMode=BMM_Replace
    rotationSpace=BCS_WorldSpace
    scaleMode=BMM_Additive
    translationMode=BMM_Replace
    translationSpace=BCS_WorldSpace
    alphaBoolBlend {blendInTime=1, blendOption=Linear, blendOutTime=1}
    boneToModify {boneName=spine_01}
    rotation {roll=20}
   }
   n14 = "Transform (Modify) Bone - Bone: spine_04"(
    ComponentPose=@n13
    Alpha=1
    Translation=Get spine_04 translation() -> (spine_04 translation:Vector)
    Rotation=Get spine_04 rotation() -> (spine_04 rotation:Rotator)
    Scale="1,1,1"
   ) {
    rotationMode=BMM_Replace
    rotationSpace=BCS_WorldSpace
    scaleMode=BMM_Additive
    translationMode=BMM_Replace
    translationSpace=BCS_WorldSpace
    alphaBoolBlend {blendInTime=1, blendOption=Linear, blendOutTime=1}
    boneToModify {boneName=spine_04}
    rotation {roll=20}
   }
   n15 = "Constraint - pack (RR)"!(ComponentPose=@n14, bAlphaBoolEnabled:bool=Get Large_body() -> (Large_body:bool), ConstraintWeights_0=1, ConstraintWeights_1:float=0) -> (Pose) {
    alphaInputType=Bool
    alphaBoolBlend {blendOption=Linear}
    boneToModify {boneName=pack}
    constraintSetup [{transformType=Rotation, targetBone {boneName=spine_04}}, {transformType=Rotation, targetBone {boneName=spine_01}}]
    constraintWeights [1,1]
   }
   n16 = "Constraint - pack (PP)"!(
    ComponentPose=@n15
    bAlphaBoolEnabled:bool=Get NOT_large_body() -> (NOT_large_body:bool)
    ConstraintWeights_0=0.7
    ConstraintWeights_1:float=0.3
   ) -> (Pose) {
    alphaInputType=Bool
    alphaBoolBlend {blendOption=Linear}
    boneToModify {boneName=pack}
    constraintSetup [{targetBone {boneName=spine_04}}, {targetBone {boneName=spine_01}}]
    constraintWeights [1,1]
   }
   n17 = "Transform (Modify) Bone - Bone: pack"!(ComponentPose=@n16, bAlphaBoolEnabled:bool=Get Large_body(), Translation="0,0,0", Rotation="0,0,10", Scale) -> (Pose) {
    alphaInputType=Bool
    rotationMode=BMM_Additive
    rotationSpace=BCS_BoneSpace
    translationMode=BMM_Additive
    translationSpace=BCS_BoneSpace
    alphaBoolBlend {blendInTime=0.1, blendOption=Custom, blendOutTime=0.2}
    boneToModify {boneName=pack}
    rotation {roll=10}
   }
   n18 = "Transform (Modify) Bone - Bone: pack"(ComponentPose=@n17, bAlphaBoolEnabled=Get Large_body_move() -> (Large_body_move:bool), Translation="0,-1.5,0", Rotation="0,0,5") {
    alphaInputType=Bool
    rotationMode=BMM_Additive
    rotationSpace=BCS_BoneSpace
    translationMode=BMM_Additive
    translationSpace=BCS_BoneSpace
    alphaBoolBlend {blendInTime=0.1, blendOption=Custom, blendOutTime=0.2}
    boneToModify {boneName=pack}
    rotation {roll=5}
    translation {y=-1.5}
   }
   Save cached pose 'Presim'(Pose=Component To Local(ComponentPose=@n18)) {cacheName=Presim}
   n19 = "Transform (Modify) Bone - Bone: pack"!(
    ComponentPose=Local To Component(LocalPose=Use cached pose 'Presim'())
    Alpha:float
    Translation="-0.842,-0.7,-6.12"
    Rotation="8.813,0.079,0"
    Scale
   ) -> (Pose) {
    rotationMode=BMM_Additive
    rotationSpace=BCS_BoneSpace
    translationMode=BMM_Additive
    translationSpace=BCS_BoneSpace
    alphaBoolBlend {blendInTime=0.1, blendOption=Custom, blendOutTime=0.2}
    boneToModify {boneName=pack}
    rotation {pitch=8.813, yaw=0.079}
    translation=<-0.842 -0.7 -6.12>
   }
   n20 = "Transform (Modify) Bone - Bone: pack"!(ComponentPose=@n19, bAlphaBoolEnabled:bool=Get bIsSkydiveFloating(), Translation="0,-1.5,0.975", Rotation="-1.111,-0.346,2", Scale) -> (Pose) {
    alphaInputType=Bool
    rotationMode=BMM_Additive
    rotationSpace=BCS_BoneSpace
    translationMode=BMM_Additive
    translationSpace=BCS_BoneSpace
    alphaBoolBlend {blendInTime=0.1, blendOption=ExpOut, blendOutTime=0.2}
    boneToModify {boneName=pack}
    rotation=<rot -1.111 -0.346 2>
    translation {y=-1.5, z=0.975}
   }
   n21 = "Transform (Modify) Bone - Bone: pack"!(
    ComponentPose=@n20
    Alpha:float=Get Right clavicle back alpha() -> (Right clavicle back alpha:double)
    Translation="0,0,0"
    Rotation="-1.695,-3.788,5.845"
    Scale
   ) -> (Pose) {
    rotationMode=BMM_Additive
    rotationSpace=BCS_BoneSpace
    translationMode=BMM_Additive
    translationSpace=BCS_BoneSpace
    alphaBoolBlend {blendInTime=0.2, blendOption=ExpOut, blendOutTime=0.2}
    alphaScaleBiasClamp {bInterpResult=true}
    boneToModify {boneName=pack}
    rotation=<rot -1.695 -3.788 5.8449>
   }
   n22 = "Transform (Modify) Bone - Bone: pack"(
    ComponentPose=@n21
    Alpha=Get Left clavicle back alpha() -> (Left clavicle back alpha:double)
    Translation="-0.074,-1,0.017"
    Rotation="4.660,-8.085,-3.276"
   ) {
    rotationMode=BMM_Additive
    rotationSpace=BCS_BoneSpace
    translationMode=BMM_Additive
    translationSpace=BCS_BoneSpace
    alphaBoolBlend {blendInTime=0.2, blendOption=ExpOut, blendOutTime=0.2}
    alphaScaleBiasClamp {bInterpResult=true}
    boneToModify {boneName=pack}
    rotation=<rot 4.66 -8.085 -3.2761>
    translation=<-0.074 -1 0.017>
   }
   Save cached pose 'PreSim2'(Pose=Blend Poses by bool(
     BlendPose_0=Component To Local(ComponentPose=@n22)
     BlendPose_1=Use cached pose 'Presim'()
     bActiveValue=Get SnowSoldierFashion character() -> (SnowSoldierFashion character:bool)
    ) {blendType=Linear}) {cacheName=PreSim2}
  ]}
 EventGraph {nodes [
   Event BlueprintUpdateAnimation()
   > n1 = Sequence!(execute) -> (then_0, then_1, then_2:Exec)
   n2 = Get BodyMesh() -> (BodyMesh:Skeletal Mesh Component Object Reference)
   > Set Left clavicle back alpha(Left clavicle back alpha:double=GetAlphaFromReferenceTransformDelta(SkeletalMeshComponent=@n2, BoneName=clavicle_l, BaseName=spine_05, InputRange=(X=-20,Y=-8), OutputRange=(X=1,Y=0), TwistAxis="1,0,0")) -> (Output_Get:double)
   > Set Right clavicle back alpha(Right clavicle back alpha:double=GetAlphaFromReferenceTransformDelta(SkeletalMeshComponent=@n2, BoneName=clavicle_r, BaseName=spine_05, InputRange=(X=-20,Y=-10), OutputRange=(X=1,Y=0), TwistAxis="1,0,0")) -> (Output_Get:double)
   n3 = Get BodyMesh()
   > Set SnowSoldierFashion character(SnowSoldierFashion character:bool=Contains(SearchIn=GetDisplayName(Object=@n3), Substring=SnowSoldierFashion)) -> (Output_Get:bool)
   Is Valid(exec=@n1.then_1, InputObject=Get BodyMesh())
   n4 = BreakTransform(InTransform=GetSocketTransform(self=Get BodyMesh(), InSocketName=spine_01, TransformSpace=RTS_World))
   > Set spine_01 translation(spine_01 translation:Vector=@n4) -> (Output_Get:Vector)
   > Set spine_01rotation(spine_01rotation:Rotator=@n4.Rotation) -> (Output_Get:Rotator)
   n5 = BreakTransform(InTransform=GetSocketTransform(self=Get BodyMesh(), InSocketName=spine_04, TransformSpace=RTS_World))
   > Set spine_04 translation(spine_04 translation:Vector=@n5) -> (Output_Get:Vector)
   > Set spine_04 rotation(spine_04 rotation:Rotator=@n5.Rotation) -> (Output_Get:Rotator)
   > n6 = Gravity Override State Function DEPRECATED(CharacterPartAnimInstance=Self-Reference(), GameTimeInSeconds=GetGameTimeInSeconds(), GravityOverrideParams=Get Gravity Override Params())
   > Set Gravity Override(Gravity Override=@n6)
   n7 = Event OnCharacterCustomizationLoaded()
   > Set BodyMesh(BodyMesh=Get Array Item with EFort Custom Part Type(Array=@n7.CustomMeshParts))
   n8 = "Equal (Enum)"(
    A:EFortCustomBodyType Enum=Get BodyTypesPermitted!(self:Custom Character Part Object Reference=Get Array Item with EFort Custom Part Type(Array:Array of Custom Character Part Object References=@n7.CustomParts) -> (Output:Custom Character Part Object Reference)) -> (BodyTypesPermitted:EFortCustomBodyType Enum=NONE)
    B:EFortCustomBodyType Enum=Large
   )
   > Set Large_body_move(Large_body_move:bool=AND Boolean(A=@n8, B=Get bIsMoving2D())) -> (Output_Get:bool)
   > Set Large_body(Large_body:bool=@n8) -> (Output_Get:bool)
   > Set NOT_large_body(NOT_large_body:bool=NOT Boolean(A=Get Large_body())) -> (Output_Get:bool)
   Reroute Node()
   n9 = GetOwnerSkeletalMeshComponent()
   n10 = BreakRotator(InRot=BreakTransform(InTransform=GetDeltaTransformFromRefPose(self=@n9, BoneName=spine_05, BaseName=spine_04)).Rotation)
   > Set backpack_rotate(backpack_rotate=MakeRotator(Roll=MapRangeClamped(Value=@n10.Yaw, InRangeA=0, InRangeB=-30, OutRangeB=25)))
   > Set backpack_translation(backpack_translation=MakeVector(Y=MapRangeClamped(Value=@n10.Yaw, InRangeA=0, InRangeB=-30, OutRangeA=-1, OutRangeB=-5)))
   TryGetPawnOwner()
   GetAlphaFromReferenceTransformDelta(BoneName=clavicle_r, BaseName=spine_05, InputRange=(X=0,Y=10), TwistAxis="1,0,0")
  ]}
]
variables [
 backpack_rotate {type=Rotator}
 backpack_translation {type=Vector}
 BodyMesh {type=SkeletalMeshComponent}
 Gravity Override {type=Vector}
 Gravity Override Params {type=GravityOverrideParamsStruct}
 Large_body {type=boolean}
 Large_body_move {type=boolean}
 Left clavicle back alpha {type=number}
 NewVar {type=boolean}
 NOT_large_body {type=boolean}
 Output {type=SkeletalMeshComponentBudgeted}
 Output_0 {type=CustomCharacterPart, default=null}
 Right clavicle back alpha {type=number}
 SnowSoldierFashion character {type=boolean}
 spine_01 translation {type=Vector}
 spine_01rotation {type=Rotator}
 spine_04 rotation {type=Rotator}
 spine_04 translation {type=Vector}
]
